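# Looping ambient sound cue for the Fried Nexus containment field cone spot at the TechHQ POI, capped at three concurrent instances.
# Three layered drone and buzz loops feed two mixed attenuation paths, one cone-shaped and one default-shaped with occlusion, both with reverb sends.
# A random node picks one of three branches of nested looping envelopers, two delayed, that slowly drift pitch and volume over a shared modulator to vary each instance.
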SoundCue /Hera/Audio/POI/TechHQ/Spots/SC_Spot_FriedNexus_Containment_Field_Cone_Loop_Cue
root=QualityLevel_0
nodes [
 {
  id=Attenuation_0
  children [Random_2]
  properties {
   bOverrideAttenuation=true
   attenuationOverrides {
    attenuationShape=Cone
    bAttenuateWithLPF=true
    bEnableListenerFocus=true
    bEnableLogFrequencyScaling=true
    bEnableSubmixSends=true
    falloffDistance=400
    focusAzimuth=25
    focusDistanceScale=0.85
    hPFFrequencyAtMax=300
    lPFFrequencyAtMax=1000
    lPFRadiusMax=2000
    lPFRadiusMin=650
    nonFocusAzimuth=55
    nonFocusVolumeAttenuation=0
    attenuationShapeExtents=<950 60 45>
    submixSendSettings [{
      bEnableHPFCutoff=false
      bEnableLPFCutoff=false
      disableManualSendClamp=false
      hPFCutoff=20
      lPFCutoff=20000
      maxSendDistance=2000
      maxSendLevel=0.6
      minSendDistance=250
      minSendLevel=0.1
      sendLevel=0.2
      sendLevelControlMethod=Linear
      soundSubmix=&Fortnite_Reverb_Default
     }]
   }
  }
 }
 {
  id=Attenuation_2
  children [Random_2]
  properties {
   bOverrideAttenuation=true
   attenuationOverrides {
    bAttenuateWithLPF=true
    bEnableListenerFocus=true
    bEnableLogFrequencyScaling=true
    bEnableOcclusion=true
    bEnableSubmixSends=true
    coneSphereFalloffDistance=950
    coneSphereRadius=250
    falloffDistance=600
    focusAzimuth=25
    focusDistanceScale=0.9
    hPFFrequencyAtMax=300
    lPFFrequencyAtMax=1000
    lPFRadiusMax=2000
    lPFRadiusMin=650
    nonFocusAzimuth=55
    nonFocusVolumeAttenuation=0.4
    occlusionInterpolationTime=0.2
    occlusionLowPassFilterFrequency=2000
    occlusionTraceChannel=ECC_GameTraceChannel11
    attenuationShapeExtents=<600 46.4459 35.9857>
    submixSendSettings [{
      bEnableHPFCutoff=false
      bEnableLPFCutoff=false
      disableManualSendClamp=false
      hPFCutoff=20
      lPFCutoff=20000
      maxSendDistance=2000
      maxSendLevel=0.6
      minSendDistance=250
      minSendLevel=0.1
      sendLevel=0.2
      sendLevelControlMethod=Linear
      soundSubmix=&Fortnite_Reverb_Default
     }]
   }
  }
 }
 {id=Delay_0, children [Enveloper_2], properties {delayMax=0.4, delayMin=0.4}}
 {id=Delay_1, children [Enveloper_3], properties {delayMax=0.25, delayMin=0.25}}
 {
  id=Enveloper_0
  children [Modulator_0]
  properties {
   bLoop=true
   bLoopIndefinitely=true
   loopEnd=60
   pitchCurve {editorCurveData {keys [
      {interpMode=RCIM_Cubic, time=0}
      <at 15.9468 0.95 RCIM_Cubic>
      {arriveTangent=0.0034, interpMode=RCIM_Cubic, leaveTangent=0.0034, time=30}
      <at 45 1.05 RCIM_Cubic>
      {interpMode=RCIM_Cubic, time=60}
     ]}}
   volumeCurve {editorCurveData {keys [<at 0 1>]}}
  }
 }
 {
  id=Enveloper_1
  children [Modulator_0]
  properties {
   bLoop=true
   bLoopIndefinitely=true
   loopEnd=30
   pitchCurve {editorCurveData {keys [{interpMode=RCIM_Cubic, time=0}, <at 15.9468 1.05 RCIM_Cubic>, {interpMode=RCIM_Cubic, time=30}]}}
   volumeCurve {editorCurveData {keys [<at 0 1>]}}
  }
 }
 {
  id=Enveloper_2
  children [Enveloper_1]
  properties {
   bLoop=true
   bLoopIndefinitely=true
   loopEnd=20
   pitchCurve {editorCurveData {keys [{time=0}]}}
   volumeCurve {editorCurveData {keys [
      {arriveTangent=-0.01, interpMode=RCIM_Cubic, time=0, value=1}
      <at 10 1.1 RCIM_Cubic>
      {interpMode=RCIM_Cubic, leaveTangent=0.01, time=20, value=1}
     ]}}
  }
 }
 {
  id=Enveloper_3
  children [Enveloper_0]
  properties {
   bLoop=true
   bLoopIndefinitely=true
   loopEnd=20
   pitchCurve {editorCurveData {keys [{time=0}]}}
   volumeCurve {editorCurveData {keys [
      {arriveTangent=-0.01, interpMode=RCIM_Cubic, time=0, value=1}
      <at 10 0.95 RCIM_Cubic>
      {interpMode=RCIM_Cubic, leaveTangent=0.01, time=20, value=1}
     ]}}
  }
 }
 {
  id=Enveloper_4
  children [Modulator_0]
  properties {
   bLoop=true
   bLoopIndefinitely=true
   loopEnd=30
   pitchCurve {editorCurveData {keys [{interpMode=RCIM_Cubic, time=0}, <at 15.9468 1.1 RCIM_Cubic>, {interpMode=RCIM_Cubic, time=30}]}}
   volumeCurve {editorCurveData {keys [<at 0 1>]}}
  }
 }
 {
  id=Enveloper_5
  children [Enveloper_4]
  properties {
   bLoop=true
   bLoopIndefinitely=true
   loopEnd=20
   pitchCurve {editorCurveData {keys [{time=0}]}}
   volumeCurve {editorCurveData {keys [
      {arriveTangent=-0.01, interpMode=RCIM_Cubic, time=0, value=1}
      <at 10 0.95 RCIM_Cubic>
      {interpMode=RCIM_Cubic, leaveTangent=0.01, time=20, value=1}
     ]}}
  }
 }
 {id=Mixer_0, children [WavePlayer_12, WavePlayer_0, WavePlayer_1], properties {inputVolume [0.5,1,0.5]}}
 {id=Mixer_2, children [Attenuation_0, Attenuation_2], properties {inputVolume [1,1]}}
 {id=Modulator_0, children [Mixer_2]}
 {id=QualityLevel_0, children [Random_0,null,null,null]}
 {id=Random_0, children [Delay_0,Delay_1,Enveloper_5], properties {weights [1,1,1]}}
 {id=Random_2, children [Mixer_0], properties {weights [1]}}
 {id=WavePlayer_0, properties {bLooping=true, soundWaveAssetPtr=/AmbientAudioContent/Waves/POI/Spot_NeonCity_Hologram_Giant_Hum_Buzz_Loop_01}}
 {id=WavePlayer_1, properties {bLooping=true, soundWaveAssetPtr=/AmbientAudioContent/Waves/SpottedLoops/spot_neon_sign_buzz_loop_01}}
 {id=WavePlayer_12, properties {bLooping=true, soundWaveAssetPtr=/AmbientAudioContent/Waves/SpottedLoops/Spot_Console_Tech_Drone_Hum_Loop_01}}
]
properties {
 bOverrideConcurrency=true
 maxDistance=1350
 soundClassObject=&Fort_Ambient_Props_Athena
 volumeMultiplier=0.25
 attenuationOverrides {attenuationShape=Cone, bAttenuate=false, coneSphereFalloffDistance=650, coneSphereRadius=250, attenuationShapeExtents=<1322.7688 45 70>}
 concurrencyOverrides {maxCount=3}
}
