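# Looping ambient spot cue for a Fried Nexus floor power conduit at the TechHQ POI, limited to one instance.
# It mixes a random electrical-unit loop and a random transformer or hologram hum, both pitched down, with a Zeropoint energy loop; the hologram wave player alone is not set to loop.
# An Attenuation node adds cone falloff, low-pass filtering, listener focus and reverb send, feeding only the first two Quality Level slots.
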SoundCue /Hera/Audio/POI/TechHQ/Spots/SC_Spot_FriedNexus_PowerConduit_Floor_Cone_Loop_Cue
root=QualityLevel_0
nodes [
 {
  id=Attenuation_0
  children [Mixer_1]
  properties {
   bOverrideAttenuation=true
   attenuationOverrides {
    attenuationShape=Cone
    bAttenuateWithLPF=true
    bEnableListenerFocus=true
    bEnableLogFrequencyScaling=true
    bEnableSubmixSends=true
    falloffDistance=500
    focusAzimuth=25
    focusDistanceScale=0.85
    hPFFrequencyAtMax=100
    lPFFrequencyAtMax=100
    lPFRadiusMax=2000
    lPFRadiusMin=100
    nonFocusAzimuth=55
    nonFocusVolumeAttenuation=0.4
    attenuationShapeExtents=<500 90>
    submixSendSettings [{
      bEnableHPFCutoff=false
      bEnableLPFCutoff=false
      disableManualSendClamp=false
      hPFCutoff=20
      lPFCutoff=20000
      maxSendDistance=2000
      maxSendLevel=0.6
      minSendDistance=250
      minSendLevel=0.1
      sendLevel=0.2
      sendLevelControlMethod=Linear
      soundSubmix=&Fortnite_Reverb_Default
     }]
   }
  }
 }
 {id=Mixer_1, children [Modulator_0, Modulator_1, WavePlayer_10], properties {inputVolume [2,0.5,1.1]}}
 {id=Modulator_0, children [Random_1], properties {pitchMax=0.8, pitchMin=0.8}}
 {id=Modulator_1, children [Random_3], properties {pitchMax=0.8, pitchMin=0.8}}
 {id=QualityLevel_0, children [Attenuation_0, Attenuation_0, null, null]}
 {id=Random_1, children [WavePlayer_5,WavePlayer_6,WavePlayer_7], properties {weights [1,1,1]}}
 {id=Random_3, children [WavePlayer_8, WavePlayer_9, WavePlayer_11], properties {weights [1,1,1]}}
 {
  id=WavePlayer_10
  properties {bLooping=true, soundWaveAssetPtr=/AmbientAudioContent/Waves/SpottedLoops/Spot_Zeropoint_Petrified_Low_Energy_Very_Close_Loop}
 }
 {id=WavePlayer_11, properties {soundWaveAssetPtr=/AmbientAudioContent/Waves/POI/Spot_NeonCity_Hologram_Giant_Hum_Buzz_Loop_01}}
 {id=WavePlayer_5, properties {bLooping=true, soundWaveAssetPtr=/AmbientAudioContent/Waves/SpottedLoops/Spot_Electrical_Power_Unit_Loop_03}}
 {id=WavePlayer_6, properties {bLooping=true, soundWaveAssetPtr=/AmbientAudioContent/Waves/SpottedLoops/Spot_Electrical_Power_Unit_Loop_02}}
 {id=WavePlayer_7, properties {bLooping=true, soundWaveAssetPtr=/AmbientAudioContent/Waves/SpottedLoops/Spot_Electrical_Power_Unit_Loop_01}}
 {id=WavePlayer_8, properties {bLooping=true, soundWaveAssetPtr=/AmbientAudioContent/Waves/SpottedLoops/Spot_Power_Transformer_Loop_01}}
 {id=WavePlayer_9, properties {bLooping=true, soundWaveAssetPtr=/AmbientAudioContent/Waves/SpottedLoops/Spot_Power_Transformer_Loop_02}}
]
properties {
 bOverrideConcurrency=true
 maxDistance=1000
 soundClassObject=&Fort_Ambient_Outdoors_Athena
 volumeMultiplier=0.15
 attenuationOverrides {attenuationShape=Cone, bAttenuate=false, coneSphereFalloffDistance=650, coneSphereRadius=250, attenuationShapeExtents=<1322.7688 45 70>}
 concurrencyOverrides {maxCount=1, resolutionRule=PreventNew, voiceStealReleaseTime=0.2}
}
