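# Ambient outdoor spot cue for a looping energy drone sphere at the Tech HQ POI, limited to one instance.
# Crystal ambience and energy-pool edge loops are mixed, then blended with a close low-end loop and fed identically to all four quality levels.
# Each layer has its own HRTF attenuation with listener focus and low-pass filtering.
# The crystal layer sends to distant reverb, while the two capsule-shaped layers send to default reverb.
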SoundCue /Hera/Audio/POI/TechHQ/Spots/SC_Spot_FriedNexus_Tech_Energy_Drone_Sphere_Loop_Cue
root=QualityLevel_0
nodes [
 {
  id=Attenuation_0
  children [WavePlayer_0]
  properties {
   bOverrideAttenuation=true
   attenuationOverrides {
    attenuationShape=Capsule
    bAttenuateWithLPF=true
    bEnableFocusInterpolation=true
    bEnableListenerFocus=true
    bEnableLogFrequencyScaling=true
    bEnableSubmixSends=true
    focusReleaseInterpSpeed=0.75
    hPFFrequencyAtMax=300
    lPFFrequencyAtMax=5000
    lPFRadiusMax=2000
    lPFRadiusMin=1000
    nonFocusVolumeAttenuation=0.75
    spatializationAlgorithm=SPATIALIZATION_HRTF
    stereoSpread=0
    attenuationShapeExtents=<650 1000>
    submixSendSettings [{
      bEnableHPFCutoff=false
      bEnableLPFCutoff=false
      disableManualSendClamp=false
      hPFCutoff=20
      lPFCutoff=20000
      maxSendDistance=2500
      maxSendLevel=0.4
      minSendDistance=1000
      minSendLevel=0.1
      sendLevel=0.2
      sendLevelControlMethod=Linear
      soundSubmix=&Fortnite_Reverb_Default
     }]
   }
  }
 }
 {
  id=Attenuation_1
  children [WavePlayer_6]
  properties {
   bOverrideAttenuation=true
   attenuationOverrides {
    bAttenuateWithLPF=true
    bEnableFocusInterpolation=true
    bEnableListenerFocus=true
    bEnableLogFrequencyScaling=true
    bEnableSubmixSends=true
    focusReleaseInterpSpeed=0.5
    focusVolumeAttenuation=1.1
    hPFFrequencyAtMax=300
    lPFFrequencyAtMax=45
    lPFRadiusMax=3000
    lPFRadiusMin=500
    nonFocusVolumeAttenuation=0.75
    spatializationAlgorithm=SPATIALIZATION_HRTF
    stereoSpread=0
    attenuationShapeExtents {x=800}
    submixSendSettings [{
      bEnableHPFCutoff=false
      bEnableLPFCutoff=false
      disableManualSendClamp=false
      hPFCutoff=20
      lPFCutoff=20000
      maxSendDistance=2500
      maxSendLevel=0.4
      minSendDistance=1000
      minSendLevel=0.1
      sendLevel=0.2
      sendLevelControlMethod=Linear
      soundSubmix=&Fortnite_Reverb_Distant
     }]
   }
  }
 }
 {
  id=Attenuation_3
  children [WavePlayer_3]
  properties {
   bOverrideAttenuation=true
   attenuationOverrides {
    attenuationShape=Capsule
    bAttenuateWithLPF=true
    bEnableFocusInterpolation=true
    bEnableListenerFocus=true
    bEnableLogFrequencyScaling=true
    bEnableSubmixSends=true
    distanceAlgorithm=Logarithmic
    focusReleaseInterpSpeed=0.75
    hPFFrequencyAtMax=300
    lPFFrequencyAtMax=5000
    lPFRadiusMax=2000
    lPFRadiusMin=1000
    nonFocusVolumeAttenuation=0
    spatializationAlgorithm=SPATIALIZATION_HRTF
    stereoSpread=0
    attenuationShapeExtents=<650 350>
    submixSendSettings [{
      bEnableHPFCutoff=false
      bEnableLPFCutoff=false
      disableManualSendClamp=false
      hPFCutoff=20
      lPFCutoff=20000
      maxSendDistance=2500
      maxSendLevel=0.4
      minSendDistance=1000
      minSendLevel=0.1
      sendLevel=0.2
      sendLevelControlMethod=Linear
      soundSubmix=&Fortnite_Reverb_Default
     }]
   }
  }
 }
 {id=Mixer_0, children [Attenuation_1, Attenuation_0], properties {inputVolume [0.7,1.5]}}
 {id=Mixer_2, children [Mixer_0, Attenuation_3], properties {inputVolume [0.7,0.45]}}
 {id=QualityLevel_0, children [Mixer_2,Mixer_2,Mixer_2,Mixer_2]}
 {id=WavePlayer_0, properties {bLooping=true, soundWaveAssetPtr=/AmbientAudioContent/Waves/SpottedLoops/Spot_Energy_Pool_Edge_Loop}}
 {
  id=WavePlayer_3
  properties {bLooping=true, soundWaveAssetPtr=/AmbientAudioContent/Waves/POI/Spot_ZP_Medium_Activity_Low_End_Close_Loop_Mono}
 }
 {id=WavePlayer_6, properties {bLooping=true, soundWaveAssetPtr=/AmbientAudioContent/Waves/POI/Spot_ZP_Crystal_Amb_Loop_03}}
]
properties {
 bOverrideConcurrency=true
 maxDistance=4600
 soundClassObject=&Fort_Ambient_Outdoors_Athena
 attenuationOverrides {
  bAttenuateWithLPF=true
  bEnableFocusInterpolation=true
  bEnableListenerFocus=true
  distanceAlgorithm=Custom
  falloffDistance=5500
  focusAttackInterpSpeed=0.5
  focusAzimuth=15
  focusDistanceScale=0.5
  focusReleaseInterpSpeed=0.25
  lPFFrequencyAtMax=500
  lPFRadiusMax=2000
  lPFRadiusMin=1000
  nonFocusAzimuth=90
  nonFocusVolumeAttenuation=0.1
  stereoSpread=2048
  attenuationShapeExtents {x=1024}
  customAttenuationCurve {editorCurveData {keys [<at 0 1>,<at 0.3 1>,<at 0.4 0.5>,<at 1 0>]}}
 }
 concurrencyOverrides {maxCount=1, resolutionRule=PreventNew}
}
