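# Looping ambient spot cue for the occluded particle accelerator tube at the TechHQ POI, routed to the Athena ambient props sound class.
# A quality-level node feeds every tier from one looping wave player using a mono low-end accelerator window sound.
# Capsule attenuation override adds listener focus and a distance-scaled linear send to the distant reverb submix.
# A concurrency override caps it at two instances, stopping the farthest first.
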SoundCue /Hera/Audio/POI/TechHQ/Spots/SC_Spot_Particle_Accelerator_Tube_Occluded_Loop_Cue
root=QualityLevel_0
nodes [
 {id=QualityLevel_0, children [WavePlayer_0,WavePlayer_0,WavePlayer_0,WavePlayer_0]}
 {
  id=WavePlayer_0
  properties {bLooping=true, soundWaveAssetPtr=&/Spot_Energy_Particle_Accelerator_Window_Low_End_Loop_Mono}
 }
]
properties {
 bDebug=true
 bOverrideAttenuation=true
 bOverrideConcurrency=true
 maxDistance=1271
 soundClassObject=&Fort_Ambient_Props_Athena
 volumeMultiplier=1
 attenuationOverrides {
  attenuationShape=Capsule
  bEnableFocusInterpolation=true
  bEnableListenerFocus=true
  bEnableSubmixSends=true
  falloffDistance=850
  nonFocusAzimuth=180
  nonFocusVolumeAttenuation=0.4
  attenuationShapeExtents=<421 250>
  submixSendSettings [{
    bEnableHPFCutoff=false
    bEnableLPFCutoff=false
    disableManualSendClamp=false
    hPFCutoff=20
    lPFCutoff=20000
    maxSendDistance=1000
    maxSendLevel=0.25
    minSendDistance=150
    minSendLevel=0
    sendLevel=0.2
    sendLevelControlMethod=Linear
    soundSubmix=&Fortnite_Reverb_Distant
   }]
 }
 concurrencyOverrides {maxCount=2, resolutionRule=StopFarthestThenPreventNew}
 concurrencySet [/AmbientAudioBase/Concurrency/AmbientConc_3Max_Spot_Default_Concurrency]
}
