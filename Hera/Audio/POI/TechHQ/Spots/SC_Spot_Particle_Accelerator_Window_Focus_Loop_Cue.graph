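# Looping ambient spot cue for the particle accelerator window at the TechHQ POI, on the Athena ambient props sound class.
# A mixer layers a looping wet energy wave with a high-end stereo wave whose player is not set to loop; all quality levels share this mix.
# Overridden attenuation uses short-range logarithmic falloff, distance low-pass filtering and interpolated listener focus that drops non-focus volume to zero.
# Concurrency limits instances, stopping the farthest first.
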SoundCue /Hera/Audio/POI/TechHQ/Spots/SC_Spot_Particle_Accelerator_Window_Focus_Loop_Cue
root=QualityLevel_0
nodes [
 {id=Mixer_0, children [WavePlayer_1,WavePlayer_4], properties {inputVolume [1,1]}}
 {id=QualityLevel_0, children [Mixer_0,Mixer_0,Mixer_0,Mixer_0]}
 {id=WavePlayer_1, properties {bLooping=true, soundWaveAssetPtr=&/Spot_Energy_Particle_Accelerator_Window_Wet_Loop}}
 {
  id=WavePlayer_4
  properties {soundWaveAssetPtr=&/Spot_Energy_Particle_Accelerator_Window_High_End_Wet_Loop_Stereo}
 }
]
properties {
 bOverrideAttenuation=true
 bOverrideConcurrency=true
 maxDistance=600
 soundClassObject=&Fort_Ambient_Props_Athena
 volumeMultiplier=0.5
 attenuationOverrides {
  bAttenuateWithLPF=true
  bEnableFocusInterpolation=true
  bEnableListenerFocus=true
  bEnableLogFrequencyScaling=true
  distanceAlgorithm=Logarithmic
  falloffDistance=500
  lPFFrequencyAtMax=300
  lPFRadiusMax=1000
  lPFRadiusMin=500
  nonFocusVolumeAttenuation=0
  stereoSpread=50
  attenuationShapeExtents=<100 50>
 }
 concurrencyOverrides {maxCount=2, resolutionRule=StopFarthestThenPreventNew}
 concurrencySet [/AmbientAudioBase/Concurrency/AmbientConc_3Max_Spot_Default_Concurrency]
}
