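# Looping ambient sound cue for server stack spot emitters at the TechHQ POI, built from hum, terminal and scanning loops.
# A quality node routes its first two levels to three mixed layers of random, modulated loops, and the last two to a simpler hum mix with random pitch ranges.
# Overrides add capsule HRTF attenuation with low-pass and custom falloff, farthest-first concurrency limiting and distance-scaled reverb sends.
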SoundCue /Hera/Audio/POI/TechHQ/Spots/SC_Spot_Server_Stack_All_Loop_Cue
root=QualityLevel_0
nodes [
 {id=Mixer_10, children [WavePlayer_15], properties {inputVolume [0.75]}}
 {id=Mixer_11, children [WavePlayer_16], properties {inputVolume [0.75]}}
 {id=Mixer_3, children [WavePlayer_11], properties {inputVolume [0.5]}}
 {id=Mixer_4, children [WavePlayer_6], properties {inputVolume [1.25]}}
 {id=Mixer_5, children [WavePlayer_7], properties {inputVolume [1.5]}}
 {id=Mixer_6, children [Modulator_0,Modulator_4,Modulator_3], properties {inputVolume [1,0.1,0.5]}}
 {id=Mixer_7, children [WavePlayer_14], properties {inputVolume [0.54]}}
 {id=Mixer_8, children [Mixer_7, WavePlayer_4, WavePlayer_13], properties {inputVolume [1,0.1,1]}}
 {id=Modulator_0, children [Random_6], properties {pitchMax=1.2, pitchMin=1}}
 {id=Modulator_1, children [Mixer_8], properties {pitchMax=1.2, pitchMin=1.1}}
 {id=Modulator_2, children [Mixer_8], properties {pitchMax=1}}
 {id=Modulator_3, children [Random_2], properties {pitchMax=1.2}}
 {id=Modulator_4, children [Random_1], properties {pitchMax=1.2, volumeMax=1.1, volumeMin=1}}
 {id=QualityLevel_0, children [Mixer_6,Mixer_6,Random_8,Random_8]}
 {id=Random_1, children [WavePlayer_9, WavePlayer_0, WavePlayer_3, WavePlayer_5, WavePlayer_10], properties {weights [1,1,1,1,1]}}
 {id=Random_2, children [WavePlayer_1, WavePlayer_12, WavePlayer_8, Mixer_10, Mixer_11], properties {weights [1,1,1,1,1]}}
 {id=Random_6, children [Mixer_3,Mixer_4,Mixer_5], properties {weights [1,1,1]}}
 {id=Random_8, children [Modulator_1,Modulator_2], properties {weights [1,1]}}
 {id=WavePlayer_0, properties {bLooping=true, soundWaveAssetPtr=/AmbientAudioContent/Waves/SpottedLoops/Spot_Server_Stack_Hum_02}}
 {
  id=WavePlayer_1
  properties {bLooping=true, soundWaveAssetPtr=/AmbientAudioContent/Waves/SpottedLoops/Spot_Console_Terminal_Communication_Hum_Loop_04}
 }
 {id=WavePlayer_10, properties {bLooping=true, soundWaveAssetPtr=/AmbientAudioContent/Waves/SpottedLoops/Spot_Server_Stack_Hum_06}}
 {
  id=WavePlayer_11
  properties {bLooping=true, soundWaveAssetPtr=/AmbientAudioContent/Waves/SpottedLoops/Spot_Computer_Server_Rack_Tower_Hum_Loop_01}
 }
 {id=WavePlayer_12, properties {bLooping=true, soundWaveAssetPtr=/AmbientAudioContent/Waves/SpottedLoops/Spot_Agency_Computer_Server_Loop}}
 {
  id=WavePlayer_13
  properties {bLooping=true, soundWaveAssetPtr=&Spot_Computer_Server_Rack_Tower_Hum_Loop_01}
 }
 {
  id=WavePlayer_14
  properties {bLooping=true, soundWaveAssetPtr=&Spot_Computer_Server_Rack_Tower_Hum_Loop_01}
 }
 {id=WavePlayer_15, properties {bLooping=true, soundWaveAssetPtr=/AmbientAudioContent/Waves/SpottedLoops/Spot_Server_Stack_Scanning_Loop_02}}
 {id=WavePlayer_16, properties {bLooping=true, soundWaveAssetPtr=/AmbientAudioContent/Waves/SpottedLoops/Spot_Server_Stack_Scanning_Loop_03}}
 {id=WavePlayer_3, properties {bLooping=true, soundWaveAssetPtr=/AmbientAudioContent/Waves/SpottedLoops/Spot_Server_Stack_Hum_04}}
 {id=WavePlayer_4, properties {bLooping=true, soundWaveAssetPtr=/AmbientAudioContent/Waves/SpottedLoops/Spot_Server_Stack_Hum_07}}
 {id=WavePlayer_5, properties {bLooping=true, soundWaveAssetPtr=/AmbientAudioContent/Waves/SpottedLoops/Spot_Server_Stack_Hum_05}}
 {
  id=WavePlayer_6
  properties {bLooping=true, soundWaveAssetPtr=/AmbientAudioContent/Waves/SpottedLoops/Spot_Console_Terminal_Communication_Hum_Loop_01}
 }
 {
  id=WavePlayer_7
  properties {bLooping=true, soundWaveAssetPtr=/AmbientAudioContent/Waves/SpottedLoops/Spot_Console_Terminal_Communication_Hum_Loop_02}
 }
 {
  id=WavePlayer_8
  properties {bLooping=true, soundWaveAssetPtr=/AmbientAudioContent/Waves/SpottedLoops/Spot_Console_Terminal_Communication_Hum_Loop_03}
 }
 {id=WavePlayer_9, properties {bLooping=true, soundWaveAssetPtr=&Spot_Server_Stack_Hum_07}}
]
properties {
 bOverrideAttenuation=true
 bOverrideConcurrency=true
 maxDistance=1650
 soundClassObject=&Fort_Ambient_Props_Athena
 subtitlePriority=5000
 volumeMultiplier=0.6
 attenuationOverrides {
  attenuationShape=Capsule
  bAttenuateWithLPF=true
  bEnableListenerFocus=true
  bEnableLogFrequencyScaling=true
  dBAttenuationAtMax=-40
  distanceAlgorithm=Custom
  falloffDistance=1350
  focusAzimuth=15
  focusReleaseInterpSpeed=0.5
  hPFFrequencyAtMax=350
  lPFFrequencyAtMax=500
  lPFRadiusMax=2000
  lPFRadiusMin=200
  nonFocusVolumeAttenuation=0.75
  occlusionInterpolationTime=0.2
  occlusionLowPassFilterFrequency=12000
  occlusionTraceChannel=ECC_GameTraceChannel11
  occlusionVolumeAttenuation=0.5
  spatializationAlgorithm=SPATIALIZATION_HRTF
  stereoSpread=100
  attenuationShapeExtents {x=300, z=90}
  customAttenuationCurve {editorCurveData {keys [
     <at 0 1>
     <at 0.1 1>
     {arriveTangent=-1.1111, interpMode=RCIM_Cubic, leaveTangent=-1.1111, time=0.2995, value=0.5132}
     {arriveTangent=-1.448, interpMode=RCIM_Cubic, leaveTangent=-1.448, tangentMode=RCTM_User, time=1, value=0}
    ]}}
 }
 concurrencyOverrides {maxCount=5, resolutionRule=StopFarthestThenPreventNew}
 soundSubmixSends [
  {
   maxSendDistance=1700
   maxSendLevel=0.6
   minSendDistance=175
   minSendLevel=0.25
   sendLevelControlMethod=Linear
   soundSubmix=&Fortnite_Reverb_Default
  }
  {
   maxSendDistance=1700
   maxSendLevel=0.5
   minSendDistance=850
   minSendLevel=0
   sendLevelControlMethod=Linear
   soundSubmix=&Fortnite_Reverb_Distant
  }
 ]
}
