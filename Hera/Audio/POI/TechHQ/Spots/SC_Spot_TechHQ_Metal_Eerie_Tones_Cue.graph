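# Sound cue for one-shot eerie metallic tone spots in the TechHQ point-of-interest ambience, using the indoor ambient one-shot sound class.
# An equally weighted random pick of five waves feeds a pitch-varying Modulator.
# The root Quality Level node routes its first two slots to that Modulator and leaves the other two empty.
# It uses shared tile-ambient attenuation, halved volume, a two-instance concurrency cap and a zero-level exterior bus send.
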SoundCue /Hera/Audio/POI/TechHQ/Spots/SC_Spot_TechHQ_Metal_Eerie_Tones_Cue
root=QualityLevel_0
nodes [
 {id=Modulator_0, children [Random_0], properties {pitchMin=0.85}}
 {id=QualityLevel_0, children [Modulator_0,Modulator_0,null,null]}
 {id=Random_0, children [WavePlayer_0,WavePlayer_1,WavePlayer_2,WavePlayer_3,WavePlayer_4], properties {weights [1,1,1,1,1]}}
 {id=WavePlayer_0, properties {soundWaveAssetPtr=/AmbientAudioContent/Waves/OneShots/Spot_Metal_Eerie_Tones_1}}
 {id=WavePlayer_1, properties {soundWaveAssetPtr=/AmbientAudioContent/Waves/OneShots/Spot_Metal_Eerie_Tones_2}}
 {id=WavePlayer_2, properties {soundWaveAssetPtr=/AmbientAudioContent/Waves/OneShots/Spot_Metal_Eerie_Tones_3}}
 {id=WavePlayer_3, properties {soundWaveAssetPtr=/AmbientAudioContent/Waves/OneShots/Spot_Metal_Eerie_Tones_4}}
 {id=WavePlayer_4, properties {soundWaveAssetPtr=/AmbientAudioContent/Waves/OneShots/Spot_Metal_Eerie_Tones_5}}
]
properties {
 attenuationSettings=/Game/Sounds/Attenuation/Fort_TileAmbient_OneShot_Att
 bOverrideConcurrency=true
 maxDistance=8000
 soundClassObject=/Game/Sounds/FortSoundClassesAndModes/Fort_Ambient_OneShots_Indoors
 volumeMultiplier=0.5
 busSends [{audioBus=null, minSendLevel=0, sendLevel=0, soundSourceBus=&AmbientAudio_Exterior_SourceBus}]
 concurrencyOverrides {maxCount=2}
}
