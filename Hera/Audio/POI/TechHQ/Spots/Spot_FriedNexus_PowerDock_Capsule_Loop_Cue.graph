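# Looping ambient spot cue for the FriedNexus power dock at the TechHQ POI, a quiet electrical hum.
# It layers one of three pitch-varied power-unit loops with a random bed: either transformer buzz plus power-receiver loop, or server loop plus transformer hum.
# A 20-second looping envelope slowly dips the buzz volume.
# Overrides add capsule HRTF attenuation with occlusion, filtering, focus and limited concurrency; only the first two quality levels are wired.
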SoundCue /Hera/Audio/POI/TechHQ/Spots/Spot_FriedNexus_PowerDock_Capsule_Loop_Cue
root=QualityLevel_0
nodes [
 {
  id=Enveloper_0
  children [Modulator_1]
  properties {
   bLoop=true
   bLoopIndefinitely=true
   loopEnd=20
   pitchCurve {editorCurveData {keys [{time=0}]}}
   volumeCurve {editorCurveData {keys [
      {arriveTangent=-0.01, interpMode=RCIM_Cubic, time=0, value=1}
      <at 10 0.7 RCIM_Cubic>
      {interpMode=RCIM_Cubic, leaveTangent=0.01, time=20, value=1}
     ]}}
  }
 }
 {id=Mixer_0, children [Enveloper_0,Modulator_2], properties {inputVolume [0.4,0.25]}}
 {id=Mixer_1, children [WavePlayer_5,Modulator_3], properties {inputVolume [1.5,1.5]}}
 {id=Mixer_2, children [Modulator_4,Random_0], properties {inputVolume [1,1]}}
 {id=Modulator_0, children [Mixer_2]}
 {id=Modulator_1, children [WavePlayer_1], properties {pitchMax=0.9, pitchMin=0.8}}
 {id=Modulator_2, children [WavePlayer_2]}
 {id=Modulator_3, children [WavePlayer_6], properties {pitchMax=1, pitchMin=1}}
 {id=Modulator_4, children [Random_2], properties {pitchMax=1.2, pitchMin=1}}
 {id=QualityLevel_0, children [Modulator_0,Modulator_0,null,null]}
 {id=Random_0, children [Mixer_0,Mixer_1], properties {weights [1,1]}}
 {id=Random_2, children [WavePlayer_0,WavePlayer_3,WavePlayer_4], properties {weights [1,1,1]}}
 {id=WavePlayer_0, properties {bLooping=true, soundWaveAssetPtr=/AmbientAudioContent/Waves/SpottedLoops/Spot_Electrical_Power_Unit_Loop_03}}
 {id=WavePlayer_1, properties {bLooping=true, soundWaveAssetPtr=/AmbientAudioContent/Waves/SpottedLoops/Spot_Transformer_Large_Buzz_01}}
 {id=WavePlayer_2, properties {bLooping=true, soundWaveAssetPtr=/AmbientAudioContent/Waves/POI/Spot_Mole_AmbientLoop_Object_PowerReceivers_03}}
 {id=WavePlayer_3, properties {bLooping=true, soundWaveAssetPtr=/AmbientAudioContent/Waves/SpottedLoops/Spot_Electrical_Power_Unit_Loop_02}}
 {id=WavePlayer_4, properties {bLooping=true, soundWaveAssetPtr=/AmbientAudioContent/Waves/SpottedLoops/Spot_Electrical_Power_Unit_Loop_01}}
 {id=WavePlayer_5, properties {bLooping=true, soundWaveAssetPtr=/AmbientAudioContent/Waves/SpottedLoops/Spot_Agency_Computer_Server_Loop}}
 {id=WavePlayer_6, properties {bLooping=true, soundWaveAssetPtr=/AmbientAudioContent/Waves/SpottedLoops/Spot_Electric_Transformer_Hum_05_Loop}}
]
properties {
 bEnableBusSends=false
 bOverrideAttenuation=true
 bOverrideConcurrency=true
 maxDistance=1150
 soundClassObject=&Fort_Ambient_Props_Athena
 volumeMultiplier=0.33
 attenuationOverrides {
  attenuationShape=Capsule
  bAttenuateWithLPF=true
  bEnableListenerFocus=true
  bEnableLogFrequencyScaling=true
  bEnableOcclusion=true
  falloffDistance=750
  focusVolumeAttenuation=1.2
  hPFFrequencyAtMax=200
  lPFFrequencyAtMax=3000
  lPFRadiusMax=1000
  lPFRadiusMin=400
  nonFocusVolumeAttenuation=0.4
  occlusionInterpolationTime=0.3
  occlusionLowPassFilterFrequency=850
  occlusionTraceChannel=ECC_GameTraceChannel11
  occlusionVolumeAttenuation=0.5
  spatializationAlgorithm=SPATIALIZATION_HRTF
  attenuationShapeExtents {y=150}
 }
 concurrencyOverrides {maxCount=3, resolutionRule=StopFarthestThenPreventNew}
 concurrencySet [/AmbientAudioBase/Concurrency/AmbientConc_3Max_Spot_Default_Concurrency]
}
